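; BP_MaskFortress_Portal_Aftermath is an actor Blueprint in the portal aftermath FX, named for the MaskFortress portal.
; Its construction script sets two scalar material parameters on the CrackGlow component, controlling global opacity and depth fade distance.
; The BeginPlay, Tick and ActorBeginOverlap events are empty, so the graph holds no runtime logic.
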(event EventBeginPlay)

(event Collision|EventActorBeginOverlap
  (params (OtherActor :type Actor%20Object%20Reference)))

(event EventTick
  (params (DeltaSeconds :type Float%20%28single-precision%29)))

(fn ConstructionScript
  (bind _crackglow (Variables|Default|GetCrackGlow))
  (Rendering|Material|SetScalarParameterValueonMaterials _crackglow "Global_Opacity" 0.4)
  (Rendering|Material|SetScalarParameterValueonMaterials _crackglow "DepthFade_Distance" 2.0))
