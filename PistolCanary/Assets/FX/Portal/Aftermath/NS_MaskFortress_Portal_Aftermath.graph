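# Niagara system for the Mask Fortress portal aftermath, a magenta glow with embers, ground flow and smoke.
# CenterGlow bursts a non-expiring sprite with a light, while Embers drift on curl noise and a point force from below.
# GroundFlow sends rotating meshes radially outward, animated through dynamic material parameters.
# GPU Smoke_Roof adds distance-culled flipbook sprites with distance-based alpha; the GPU vortex rock emitter is disabled.
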NiagaraSystem /PistolCanary/Assets/FX/Portal/Aftermath/NS_MaskFortress_Portal_Aftermath
dependencies {
 usedDataInterfaces [
  &NiagaraDataInterfaceCurve
  &NiagaraDataInterfaceSpriteRendererInfo
  &NiagaraDataInterfaceColorCurve
  &NiagaraDataInterfaceGBuffer
 ]
 usedDynamicInputs []
 usedModules [
  &self:SystemSpawnScript.SystemScriptSource.SystemScriptGraph.NiagaraNodeAssignment_0.SetVariables_#1
  &SystemState
 ]
}
emitters {
 CenterGlow {
  emitterUpdateScript [
   &EmitterState(
    Inactive Response=::NewEnumerator1
    Life Cycle Mode=::NewEnumerator1
    Loop Behavior=::NewEnumerator1
    Loop Duration Mode=::NewEnumerator1
    UseLoopDelay=false
   )
   &SpawnBurst_Instantaneous(Use Spawn Probability=false, UseLoopCountLimit=false)
  ]
  particleSpawnScript [
   &InitializeParticle(
    Color=<rgba 0.9057 0.0242 1 0.6>
    Position Offset=<0 0 10>
    Sprite Size Mode=::NewEnumerator3
    Uniform Sprite Size=1200
    UsePositionOffset=true
    Write Material Random=true
   )
   &SpriteFacingAndAlignment(Sprite Alignment=<0 0 1>, Sprite Facing=<0 0 1>, WriteAlignment=false, WriteFacing=true)
   &CameraOffset(Camera Offset Amount=75)
   &DynamicMaterialParameters(
    Index 0 Param 1=3
    Index 0 Param 2=2
    Module.Index0Float0Write=true
    Module.Index0Float1Write=true
    Module.Index0Float2Write=true
    Module.Index0Float3Write=true
   )
   /SetVariables(Particles.LightPosition:NiagaraPosition=<0 0 60>) {isSetParameters=true}
  ]
  particleUpdateScript [
   &ParticleState(Kill Particles When Lifetime Has Elapsed=false)
   /SetVariables(Particles.LightColor=&Multiply_LinearColorByFloat(Linear Color=$Particles.Color, Float=7)) {isSetParameters=true}
  ]
  rendererClasses [&NiagaraSpriteRendererProperties, &NiagaraLightRendererProperties]
  renderers [
   {rendererClass=&NiagaraSpriteRendererProperties}
   {rendererClass=&NiagaraLightRendererProperties, rendererIndex=1}
  ]
 }
 Embers {
  emitterUpdateScript [
   &EmitterState()
   &SpawnRate(SpawnRate=100, Use Spawn Probability=false)
  ]
  particleSpawnScript [
   &InitializeParticle(
    Color Maximum=<rgba 5 0.2566 4.6288 1>
    Color Minimum=<rgba 0.6172 0.0238 1 1>
    Color Mode=::NewEnumerator2
    Lifetime Max=5
    Lifetime Min=3
    Lifetime Mode=::NewEnumerator1
    Sprite Size Mode=::NewEnumerator1
    Uniform Sprite Size Max=5
    Uniform Sprite Size Min=3
    UsePositionOffset=false
    Write Material Random=true
   )
   &ShapeLocation(Cylinder Radius=1000, MaskSpawnGroup=false, Override Seed=false, Shape Primitive=::NewEnumerator1, Surface Only=false, UseOwnerScale=true)
   &DynamicMaterialParameters(
    Index 0 Param 1=2
    Index 0 Param 3=1
    Module.Index0Float0Write=true
    Module.Index0Float1Write=false
    Module.Index0Float2Write=true
    Module.Index0Float3Write=false
   )
  ]
  particleUpdateScript [
   &ParticleState()
   &ScaleSpriteSize(Uniform Curve Sprite Scale=DataInterface() {propertyValues {Curve {
       defaultValue=0
       keys [
        {arriveTangent=6.7678, interpMode=RCIM_Cubic, leaveTangent=6.7678, tangentMode=RCTM_User, time=0, value=0}
        {interpMode=RCIM_Cubic, tangentMode=RCTM_User, time=0.5, value=1}
        {arriveTangent=-5.8188, tangentMode=RCTM_User, time=1, value=0}
       ]
      }}})
   &CurlNoiseForce(
    Bias Noise=false
    Noise Frequency=20
    Noise Strength=&RandomRangeFloat(Minimum=40, Maximum=80)
    Pan Noise=false
   )
   &PointForce(
    Force Origin=&AddVectorToPosition(Position=$Engine.Owner.Position, Vector=<0 0 -100>)
    Force Strength=&RandomRangeFloat(Minimum=10, Maximum=70)
    Module.Use Falloff=false
    Module.UseLinearForce=false
   )
   &Drag(Use Linear Drag=true, Use Rotational Drag=false)
   &SolveForcesAndVelocity(Clamp Velocity=false, Limit Acceleration=false)
   &ScaleSpriteSizeBySpeed(Max Scale Factor=<0.5 5>)
  ]
  rendererClasses [&NiagaraSpriteRendererProperties]
  renderers [{rendererClass=&NiagaraSpriteRendererProperties}]
 }
 GroundFlow {
  emitterSpawnScript [/SetVariables(Emitter.MeshRibbonData=DataInterface() {propertyValues {
      bExposeCurve=true
      AlphaCurve {keys []}
      BlueCurve {keys [<at 0 1>,<at 1 1>]}
      GreenCurve {keys [<at 0 -0.5 RCIM_Cubic>, {arriveTangent=0.0331, interpMode=RCIM_Cubic, leaveTangent=0.0331, tangentMode=RCTM_User, time=1, value=-1}]}
      RedCurve {keys [<at 0 1 RCIM_Cubic>,<at 1 2 RCIM_Cubic>]}
     }}) {isSetParameters=true}]
  emitterUpdateScript [
   &EmitterState()
   &SpawnBurst_Instantaneous(Spawn Count=12, Use Spawn Probability=false, UseLoopCountLimit=false) {enabled=false}
   &SpawnRate(SpawnRate=6, Use Spawn Probability=false)
  ]
  particleSpawnScript [
   &InitializeParticle(
    Color=<rgba 25 25 25 0.25>
    Lifetime=&RandomRangeFloat(Minimum=5.5, Maximum=6.5)
    Mesh Scale=&Multiply_VectorByFloat(Vector=<1 2 0.2>, Float=&RandomRangeFloat(Minimum=2, Maximum=4))
    Mesh Scale Mode=::NewEnumerator4
    UsePositionOffset=false
    Write Material Random=true
   )
   &InitialMeshOrientation(
    Mesh Orientation Mode=::NewEnumerator2
    Orientation Vector=<0 1 0>
    Rotation=&RandomRangeVector(Minimum=<0 0 0>, Maximum=<0 0 1>)
    Use Rotation Vector=true
   )
   &SystemLocation(MaskSpawnGroup=false, Offset=<0 0 -25>)
   &MeshRotationForce(Rotation=<0 0 -0.1>)
   &ApplyInitialForces()
   &AddVelocity(
    Cone Angle=180
    Cone Axis=<0 0 1>
    Inner Cone Angle=179
    Module.UseConeFalloff=true
    Module.UseVelDistribution=false
    Override Seed=false
    Velocity Mode=::NewEnumerator2
    Velocity Speed=200
   )
   &CameraOffset(Camera Offset Amount=100)
  ]
  particleUpdateScript [
   &ParticleState()
   &ScaleMeshSize(Scale Factor=&VectorFromFloat(Value=&FloatFromCurve(FloatCurve=DataInterface() {propertyValues {Curve {keys [
          {arriveTangent=0.5884, interpMode=RCIM_Cubic, leaveTangent=0.5884, tangentMode=RCTM_User, time=0, value=0.5}
          {arriveTangent=0.2883, interpMode=RCIM_Cubic, leaveTangent=0.2883, tangentMode=RCTM_User, time=1, value=1}
         ]}}})))
   &DynamicMaterialParameters(
    Index 0 Param 1=&FloatFromCurve(FloatCurve=DataInterface() {propertyValues {Curve {keys [<at 0 0>, {arriveTangent=-1.8673, interpMode=RCIM_Cubic, leaveTangent=-1.8673, tangentMode=RCTM_User, time=1, value=0.25}]}}})
    Index 0 Param 2=&FloatFromCurve(
     FloatCurve=DataInterface() {propertyValues {Curve {keys [{interpMode=RCIM_Cubic, tangentMode=RCTM_User, time=0, value=0}, {interpMode=RCIM_Cubic, tangentMode=RCTM_User, time=1, value=1}]}}}
     Scale Curve=0.5
    )
    Index 0 Param 3=&RandomRangeFloat(Minimum=5, Maximum=25)
    Index 0 Param 4=&Add_Float(A=&RandomRangeFloat(Minimum=25, Maximum=100), B=0)
    Index 1 Param 1=1
    Module.Index0Param0Write=true
    Module.Index0Param1Write=true
    Module.Index0Param2Write=true
    Module.Index0Param3Write=true
    Module.Index1Param0Write=true
    Module.Index1Param1Write=true
    Module.Index1Param2Write=false
    Module.Index1Param3Write=false
    Write Parameter Index 1=true
   )
   &ScaleColor(
    Scale Alpha=&FloatFromCurve(
     FloatCurve=DataInterface() {propertyValues {Curve {
        defaultValue=1
        keys [
         {interpMode=RCIM_Cubic, tangentMode=RCTM_User, time=0, value=0}
         {interpMode=RCIM_Cubic, tangentMode=RCTM_User, time=0.5, value=1}
         {tangentMode=RCTM_User, time=1, value=0}
        ]
       }}}
     Scale Curve=0.01
    )
    Scale RGB=<5 1 5>
    ScaleA=true
    ScaleRGB=true
   )
   &SolveForcesAndVelocity(Clamp Velocity=false, Limit Acceleration=false)
  ]
  rendererClasses [&NiagaraMeshRendererProperties]
  renderers [{rendererClass=&NiagaraMeshRendererProperties, rendererIndex=0}]
 }
 Rocks_OnSpire001 {
  bEnabled=false
  simTarget=GPUComputeSim
  emitterUpdateScript [
   &EmitterState(
    Enable Distance Culling=true
    MaxDistance=25000
    Module.bApplySpawnCountScaleByDistance=true
    Module.bUseMaxDistance=true
    Module.bUseMinDistance=false
    Scalability Mode=::NewEnumerator1
    Scale Spawn Count=true
    Spawn Count Scale By Distance=DataInterface() {propertyValues {Curve {keys [<at 0 1>,<at 0.1 1>,<at 1 1>]}}}
   )
   &SpawnRate(SpawnRate=50, Use Spawn Probability=false)
  ]
  particleSpawnScript [
   &InitializeParticle(
    Color=<rgba 0.3542 0.024 0.3365 1>
    Lifetime Max=4
    Lifetime Min=3
    Lifetime Mode=::NewEnumerator1
    Mesh Scale Max=<5 5 5>
    Mesh Scale Min=<1 1 2>
    Mesh Scale Mode=::NewEnumerator2
    Sprite Rotation Mode=::NewEnumerator2
    Sprite Size Mode=::NewEnumerator1
    Uniform Sprite Size Max=800
    Uniform Sprite Size Min=250
    UsePositionOffset=false
    Write Material Random=true
   ) {module=InitializeParticle001}
   &InitialMeshOrientation(
    Mesh Orientation Mode=::NewEnumerator2
    Rotation=&RandomRangeVector(Minimum=<-0.02 -0.02 0>, Maximum=<0.02 0.02 1>)
    Use Rotation Vector=true
   )
   &MeshRotationForce(Rotation=&RandomRangeVector(Minimum=<0 0 -0.5>, Maximum=<0.1 0.1 0.5>))
   &ApplyInitialForces()
   &CameraOffset(Camera Offset Amount=25)
   &ShapeLocation(
    MaskSpawnGroup=false
    Non Uniform Scale=<1 1 0>
    Offset=<0 0 1500>
    Offset Mode=::NewEnumerator0
    Override Seed=false
    Sphere Radius=1500
    UseOwnerScale=true
   ) {module=ShapeLocation002}
   /Niagara/Modules/Update/Utility/SampleGBuffer(Kill Beyond Depth=false, Snap To Depth:ENiagaraGBufferSnapToDepth=::NewEnumerator1, Write To Color=false, Write To Position=false)
   /SetVariables(Particles.InitialPos:NiagaraPosition=$Particles.Position) {isSetParameters=true, module=SetVariables_#2}
   &KillParticlesInVolume(Invert Volume=true, Sphere Radius=5000)
   &AddVelocity(Velocity=<0 0 -50>)
   /SetVariables(Particles.MeshIndex=&RandomRangeInt(Minimum=0, Maximum=3)) {isSetParameters=true, module=SetVariables_#3}
   &SubUVAnimation(
    Evaluation Type=::NewEnumerator0
    "Sprite Renderer "=DataInterface() {propertyValues {SpriteRenderer=&self:Rocks_OnSpire001_0.NiagaraSpriteRendererProperties_0}}
    SubUV Animation Mode:ENiagara_SubUVLookupModeV2=::NewEnumerator2
    UseEndFrame=false
    UseStartFrame=false
   )
  ]
  particleUpdateScript [
   &ParticleState()
   &UpdateMeshOrientation(Rotation Rate=&RandomRangeFloat(Minimum=-1, Maximum=1), Rotation Vector=<1 0.01 1>)
   &CurlNoiseForce(
    Bias Noise=true
    Bias Noise Field:Vector3f=<0 0 -1>
    Noise Frequency=&RandomRangeFloat(Minimum=0.1, Maximum=0.2)
    Noise Strength=&RandomRangeFloat(Minimum=25, Maximum=250)
    Pan Noise=true
    Pan Noise Field=<1 0 -1>
   )
   &PointForce(Force Falloff Distance=2500, Force Strength=-2500, Module.Use Falloff=true, Module.UseLinearForce=false)
   &VortexForce(UseOriginOffset=false, Vortex Force Amount=1500)
   &SolveForcesAndVelocity(
    Acceleration Limit=25
    Clamp Velocity=true
    Limit Acceleration=true
    Speed Limit:NiagaraFloat=&Multiply_Float(
     A=250
     B=&FloatFromCurve(FloatCurve=DataInterface() {propertyValues {Curve {keys [
          <at 0 0.5>
          {arriveTangent=1.255, interpMode=RCIM_Cubic, leaveTangent=1.255, tangentMode=RCTM_User, time=0.5, value=1}
          {arriveTangent=2.5018, interpMode=RCIM_Cubic, leaveTangent=2.5018, tangentMode=RCTM_User, time=1, value=3}
         ]}}})
    )
   )
   &CalculateAccurateVelocity()
   /SetVariables(Particles.SubImageIndex=&FloatFromCurve(FloatCurve=DataInterface() {propertyValues {Curve {keys [<at 0 0>,<at 1 1>]}}}, Scale Curve=16)) {isSetParameters=true, module=SetVariables_#4}
   &ScaleColor(
    Scale Alpha=&FloatFromCurve(
     FloatCurve=DataInterface() {propertyValues {Curve {
        defaultValue=1
        keys [
         {interpMode=RCIM_Cubic, tangentMode=RCTM_User, time=0, value=0}
         {interpMode=RCIM_Cubic, tangentMode=RCTM_User, time=0.5, value=1}
         {tangentMode=RCTM_User, time=1, value=0}
        ]
       }}}
     Scale Curve=0.05
    )
    ScaleA=true
    ScaleRGB=true
   )
   &ScaleMeshSize(Scale Factor=&Multiply_VectorByFloat(
     Vector=&Multiply_VectorByFloat(Vector=<0.04 0.04 0.04>, Float=$User.ScaleMult)
     Float=&FloatFromCurve(
      FloatCurve=DataInterface() {propertyValues {Curve {
         defaultValue=1
         keys [
          {interpMode=RCIM_Cubic, tangentMode=RCTM_User, time=0, value=0.75}
          {interpMode=RCIM_Cubic, tangentMode=RCTM_User, time=0.5, value=1}
          {tangentMode=RCTM_User, time=1, value=0}
         ]
        }}}
      Scale Curve=/Niagara/DynamicInputs/Scalability/ScalabilityDistanceBasedFloat(
       Curve:NiagaraDataInterfaceCurve=DataInterface() {propertyValues {Curve {keys [<at 0 0>,<at 0.02 1>,<at 0.8 1>,<at 1 0>]}}}
       CurveIndex:NiagaraFloat=$Emitter.Scalability.DistanceFraction
       Bias:NiagaraFloat=0
       Scale:NiagaraFloat=3
      )
     )
    ))
   &CameraOffset() {module=CameraOffset001}
   &DynamicMaterialParameters(
    Index 0 Param 4=&FloatFromCurve(FloatCurve=DataInterface() {propertyValues {Curve {keys [<at 0 1>,<at 1 0>]}}}, Scale Curve=2)
    Module.Index0Param0Write=true
    Module.Index0Param1Write=true
    Module.Index0Param2Write=true
    Module.Index0Param3Write=true
   )
  ]
  rendererClasses [&NiagaraMeshRendererProperties]
  renderers [{rendererClass=&NiagaraMeshRendererProperties, rendererIndex=0}]
 }
 Smoke_Roof {
  simTarget=GPUComputeSim
  emitterUpdateScript [
   &EmitterState(
    Enable Distance Culling=true
    MaxDistance=25000
    Module.bUseMaxDistance=true
    Module.bUseMinDistance=false
    Scalability Mode=::NewEnumerator1
   )
   &SpawnRate(SpawnRate=15, Use Spawn Probability=false)
  ]
  particleSpawnScript [
   &InitializeParticle(
    Color=<rgba 0.9167 0.0621 0.871 1>
    Lifetime Max=4
    Lifetime Min=3
    Lifetime Mode=::NewEnumerator1
    Sprite Rotation Mode=::NewEnumerator2
    Sprite Size Mode=::NewEnumerator1
    Uniform Sprite Size Max=600
    Uniform Sprite Size Min=300
    UsePositionOffset=false
    Write Material Random=true
   ) {module=InitializeParticle001}
   &InitialMeshOrientation(
    Mesh Orientation Mode=::NewEnumerator2
    Rotation=&RandomRangeVector(Minimum=<-0.02 -0.02 0>, Maximum=<0.02 0.02 1>)
    Use Rotation Vector=true
   )
   &CameraOffset(Camera Offset Amount=25)
   &ShapeLocation(
    Cylinder Height=50
    Cylinder Radius=1100
    MaskSpawnGroup=false
    Offset=<0 0 -100>
    Offset Mode=::NewEnumerator0
    Override Seed=false
    Shape Primitive=::NewEnumerator1
    Surface Only=true
    Surface Only Band Thickness=0.8
    Use Endcaps In Surface Only Mode=false
    UseOwnerScale=true
   ) {module=ShapeLocation002}
   /SetVariables(Particles.InitialPos:NiagaraPosition=$Particles.Position) {isSetParameters=true, module=SetVariables_#2}
   &AddVelocity()
   /SetVariables(Particles.MeshIndex=&RandomRangeInt(Minimum=0, Maximum=3)) {isSetParameters=true, module=SetVariables_#3}
   &SubUVAnimation(
    Evaluation Type=::NewEnumerator0
    "Sprite Renderer "=DataInterface() {propertyValues {SpriteRenderer=&self:Smoke_Roof_0.NiagaraSpriteRendererProperties_0}}
    SubUV Animation Mode:ENiagara_SubUVLookupModeV2=::NewEnumerator2
    UseEndFrame=false
    UseStartFrame=false
   )
  ]
  particleUpdateScript [
   &ParticleState()
   &UpdateMeshOrientation(Rotation Rate=&RandomRangeFloat(Minimum=-0.1, Maximum=0.02), Rotation Vector=<0 0 1>)
   &CurlNoiseForce(
    Bias Noise=true
    Bias Noise Field:Vector3f=<0 0 -1>
    Noise Frequency=&RandomRangeFloat(Minimum=0.1, Maximum=0.2)
    Noise Strength=&RandomRangeFloat(Minimum=0, Maximum=25)
    Pan Noise=true
    Pan Noise Field=<1 0 -1>
   )
   &PointForce(Force Falloff Distance=2500, Module.Use Falloff=true, Module.UseLinearForce=false)
   &SolveForcesAndVelocity(
    Acceleration Limit=25
    Clamp Velocity=true
    Limit Acceleration=true
    Speed Limit:NiagaraFloat=&Multiply_Float(
     A=2500
     B=&FloatFromCurve(FloatCurve=DataInterface() {propertyValues {Curve {keys [
          <at 0 0.5>
          {arriveTangent=1.255, interpMode=RCIM_Cubic, leaveTangent=1.255, tangentMode=RCTM_User, time=0.5, value=1}
          {arriveTangent=2.5018, interpMode=RCIM_Cubic, leaveTangent=2.5018, tangentMode=RCTM_User, time=1, value=3}
         ]}}})
    )
   )
   &CalculateAccurateVelocity()
   /SetVariables(Particles.SubImageIndex=&FloatFromCurve(FloatCurve=DataInterface() {propertyValues {Curve {keys [<at 0 0>,<at 1 1>]}}}, Scale Curve=16)) {isSetParameters=true, module=SetVariables_#4}
   &ScaleColor(
    Scale Alpha=&Multiply_Float(
     A=&FloatFromCurve(
      FloatCurve=DataInterface() {propertyValues {Curve {
         defaultValue=1
         keys [
          {interpMode=RCIM_Cubic, tangentMode=RCTM_User, time=0, value=0}
          {interpMode=RCIM_Cubic, tangentMode=RCTM_User, time=0.5, value=1}
          {tangentMode=RCTM_User, time=1, value=0}
         ]
        }}}
      Scale Curve=0.1
     )
     B=&FN_LOD_Distance_Remap(Start Distance=500, End Distance=5000, Minimum Output=0.35, Maximum Output=0.2, Clamp Output=true)
    )
    Scale RGB=&MakeVectorFromLinearColorRGB(Color=&ColorFromCurve(ColorCurve=DataInterface() {propertyValues {
        AlphaCurve {keys [<at 0 1 RCIM_Cubic>,<at 1 0 RCIM_Cubic>]}
        BlueCurve {keys [<at 0 1.25 RCIM_Cubic>,<at 0.8568 0 RCIM_Cubic>]}
        GreenCurve {keys [<at 0 1.25 RCIM_Cubic>,<at 0.8568 0 RCIM_Cubic>]}
        RedCurve {keys [<at 0 1.25 RCIM_Cubic>,<at 0.8568 0 RCIM_Cubic>]}
       }}))
    ScaleA=true
    ScaleRGB=true
   )
   &ScaleMeshSize(Scale Factor=&Multiply_VectorByFloat(
     Vector=&Multiply_VectorByFloat(Vector=<0.1 0.1 0.1>, Float=$User.ScaleMult)
     Float=&FloatFromCurve(
      FloatCurve=DataInterface() {propertyValues {Curve {
         defaultValue=1
         keys [
          {interpMode=RCIM_Cubic, tangentMode=RCTM_User, time=0, value=0}
          {interpMode=RCIM_Cubic, tangentMode=RCTM_User, time=0.5, value=1}
          {tangentMode=RCTM_User, time=1, value=0}
         ]
        }}}
      Scale Curve=&ScalabilityDistanceBasedFloat(
       Curve:NiagaraDataInterfaceCurve=DataInterface() {propertyValues {Curve {keys [<at 0 0>,<at 0.02 1>,<at 0.8 1>,<at 1 0>]}}}
       CurveIndex:NiagaraFloat=$Emitter.Scalability.DistanceFraction
       Bias:NiagaraFloat=0
       Scale:NiagaraFloat=1
      )
     )
    ))
   &CameraOffset(Camera Offset Amount=/Niagara/DynamicInputs/Math/CalculateParticleRadius(
     Radius Calculation Type:ENiagaraCollisionRadiusOptions=::NewEnumerator0
     Particle Radius:NiagaraFloat=1
     Mesh Dimensions:Vector3f=<1 1 1>
     Method for Calculating Particle Radius:ENiagaraCalculateRadiusOptions=::NewEnumerator2
    )) {module=CameraOffset001}
   &DynamicMaterialParameters(
    Index 0 Param 4=&FloatFromCurve(FloatCurve=DataInterface() {propertyValues {Curve {keys [<at 0 1>,<at 1 0>]}}}, Scale Curve=2)
    Module.Index0Param0Write=true
    Module.Index0Param1Write=true
    Module.Index0Param2Write=true
    Module.Index0Param3Write=true
   )
  ]
  rendererClasses [&NiagaraSpriteRendererProperties]
  renderers [{rendererClass=&NiagaraSpriteRendererProperties}]
 }
}
summary {
 systemName=NS_MaskFortress_Portal_Aftermath
 emitters [
  {emitterName=Smoke_Roof, simTarget=GPUComputeSim, rendererClasses [&NiagaraSpriteRendererProperties]}
  {bEnabled=false, emitterName=Rocks_OnSpire001, simTarget=GPUComputeSim, rendererClasses [&NiagaraMeshRendererProperties]}
  {emitterName=GroundFlow, rendererClasses [&NiagaraMeshRendererProperties]}
  {emitterName=Embers, rendererClasses [&NiagaraSpriteRendererProperties]}
  {emitterName=CenterGlow, rendererClasses [&NiagaraSpriteRendererProperties, &NiagaraLightRendererProperties]}
 ]
 userVariables [
  User.bIsFrontEndPreview {type=NiagaraBool, defaultValue=true}
  User.ScaleMult {type=NiagaraFloat, defaultValue=1, description=""}
  User.SpawnRatMult {type=NiagaraFloat, defaultValue=1, description=""}
 ]
}
